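# Sound cue for a close thunder strike, built as three layers mixed together.
# A randomly picked close clap with slight pitch variation plays alongside a boosted stinger, followed after a fixed half-second delay by one of two distant rumbles.
# The cue uses the Fort_Quests sound class and overrides attenuation with a wide stereo spread and manual reverb send.
# Its concurrency override caps simultaneous plays at four, stopping the quietest.
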SoundCue /FNGameplayCues/Athena/Sounds/Ambient/Thunder/Thunder_Close_Cue
root=Mixer_0
nodes [
 {id=Delay_0, children [Random_1], properties {delayMax=0.5, delayMin=0.5}}
 {id=Mixer_0, children [Modulator_0,Delay_0,WavePlayer_3], properties {inputVolume [1,1,1.4]}}
 {id=Modulator_0, children [Random_0], properties {pitchMax=1.1, pitchMin=0.9}}
 {id=Random_0, children [WavePlayer_0,WavePlayer_1], properties {weights [1,1]}}
 {id=Random_1, children [WavePlayer_2,WavePlayer_5], properties {weights [1,1]}}
 {id=WavePlayer_0, properties {soundWaveAssetPtr=&/Thunder_Close_01}}
 {id=WavePlayer_1, properties {soundWaveAssetPtr=&/Thunder_Close_02}}
 {id=WavePlayer_2, properties {soundWaveAssetPtr=&/Thunder_Distant_01}}
 {id=WavePlayer_3, properties {soundWaveAssetPtr=&/Thunder_Stinger_01}}
 {id=WavePlayer_5, properties {soundWaveAssetPtr=&/Thunder_Distant_03}}
]
properties {
 bOverrideAttenuation=true
 bOverrideConcurrency=true
 maxDistance=5400
 soundClassObject=&Fort_Quests
 volumeMultiplier=2.24
 attenuationOverrides {
  falloffDistance=4000
  manualReverbSendLevel=15
  nonSpatializedRadiusStart=1
  reverbSendMethod=Manual
  stereoSpread=8000
  attenuationShapeExtents=<1400 1500>
 }
 concurrencyOverrides {maxCount=4, resolutionRule=StopQuietest}
}
